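; Passive quest ability that detects when the player reaches max ammo, for a reach-max-ammo quest.
; On activation it starts a looping timer at a randomized interval, which is cleared when the ability ends.
; Each tick compares the player's inventory count of each listed ammo type to its max stack size to update a MaxAmmo flag.
; If the flag is set after the loop, it broadcasts the CustomVerb.MaxAmmo message for the player controller.
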(event Ability|EventFailedToActivatePassiveAbility
  (params (ActorInfo :type Gameplay%20Ability%20Actor%20Info%20Structure))
  (|CallParentFunction ActorInfo))

(event Ability|EventOnEndAbility
  (params (bWasCancelled :type Boolean))
  (|CallParentFunction bWasCancelled)
  (Utilities|Time|ClearandInvalidateTimerbyHandle (Variables|Default|GetTimer%20Handle)))

(event Ability|EventActivateAbility
  (|CallParentFunction)
  (Variables|Default|SetMaxAmmo false)
  (bind _returnvalue (Utilities|Time|SetTimerbyEvent (delegate Custom|Check%20Ammo) (Math|Random|RandomFloatinRange 1.0 1.5) true))
  (Variables|Default|SetTimer%20Handle _returnvalue))

(event Custom|Check%20Ammo
  ;[1 1 1] Check if ammo is max, if so open gate
  (bind _array_element (Utilities|Array|ForEachLoopwithBreak (Variables|Default|GetAmmo%20Types))
    (:LoopBody
      (if (>= (Inventory|GetItemQuantityonInventoryOwner (Fortnite|GetFortPlayerControllerfromActor (Ability|GetAvatarActorfromActorInfo)) _array_element) (Item|GetMaxStackSize _array_element 0))
        (Variables|Default|SetMaxAmmo true))
      (Variables|Default|SetMaxAmmo false))
    (:Completed
      _))
  (if (Variables|Default|GetMaxAmmo)
    (CustomVerb|BroadcastCustomVerbMessage (CustomVerb|MakeCustomVerbMessage (Fortnite|GetFortPlayerControllerfromActor (Variables|GAVars|Data|GetOwningPlayerPawn)) "(TagName=\"CustomVerb.MaxAmmo\")"))))
